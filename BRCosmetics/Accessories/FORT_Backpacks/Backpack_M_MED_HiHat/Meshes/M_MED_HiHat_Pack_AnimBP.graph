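# Animation Blueprint for the M_MED_HiHat backpack that offsets its pack bone and rotates a dyn_key bone.
# The AnimGraph chains additive bone-space Transform Bone nodes through cached poses: two small pack nudges, a larger pack offset, then the key rotation.
# Update derives a steadily growing roll from real time and a translation from spine_05's yaw delta against spine_01.
# On customization load, custom-part gender and body type set female medium and large fit flags.
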AnimBlueprint /BRCosmetics/Accessories/FORT_Backpacks/Backpack_M_MED_HiHat/Meshes/M_MED_HiHat_Pack_AnimBP
target_skeleton=&/M_MED_HiHat_Pack_Skeleton
graphs [
 AnimGraph {nodes [
   Output Pose(Result=Use cached pose 'Dyn_Cache'()) {layerGroup=None}
   n1 = Local To Component(LocalPose=Use cached pose 'SpineRot_Cache'())
   Save cached pose 'Dyn_Cache'(Pose=Component To Local(ComponentPose="Transform (Modify) Bone - Bone: dyn_key"!(ComponentPose=@n1, Rotation="0,0,0") -> (Pose) {rotationMode=BMM_Additive, rotationSpace=BCS_BoneSpace, boneToModify {boneName=dyn_key}})) {cacheName=Dyn_Cache}
   n2 = "Transform (Modify) Bone - Bone: pack"!(ComponentPose=Local To Component(LocalPose=Input Pose()), bAlphaBoolEnabled:bool, Translation="0,1,0") -> (Pose) {alphaInputType=Bool, translationMode=BMM_Additive, translationSpace=BCS_BoneSpace, boneToModify {boneName=pack}, translation {y=1}}
   n3 = "Transform (Modify) Bone - Bone: pack"(ComponentPose=@n2, Translation="0,0.5,0") {alphaInputType=Bool, translationMode=BMM_Additive, translationSpace=BCS_BoneSpace, boneToModify {boneName=pack}, translation {y=0.5}}
   Save cached pose 'Pos_Cache'(Pose=Component To Local(ComponentPose=@n3)) {cacheName=Pos_Cache}
   n4 = "Transform (Modify) Bone - Bone: pack"!(ComponentPose=Local To Component(LocalPose=Use cached pose 'Pos_Cache'()), Alpha:float, Translation="0,100,0", Rotation:Rotator="0,0,0") -> (Pose) {
    rotationSpace=BCS_BoneSpace
    translationMode=BMM_Additive
    translationSpace=BCS_BoneSpace
    alphaBoolBlend {blendInTime=0.3, blendOutTime=0.3}
    boneToModify {boneName=pack}
    translation {y=100}
   }
   Save cached pose 'SpineRot_Cache'(Pose=Component To Local(ComponentPose=@n4)) {cacheName=SpineRot_Cache}
  ]}
 EventGraph {nodes [
   Event BlueprintUpdateAnimation()
   n1 = float / float(A=GetRealTimeSeconds(), B=180) {is=title K2Node_CallFunction}
   > Set AmbientRot(AmbientRot:Rotator=MakeRotator(Roll=float * float(A=@n1, B=10000))) -> (Output_Get:Rotator)
   n2 = GetOwnerSkeletalMeshComponent()
   n3 = BreakRotator(InRot=BreakTransform(InTransform=GetDeltaTransformFromRefPose(self=@n2, BoneName=spine_05, BaseName=spine_01)).Rotation)
   > Set Backpack Translation(Backpack Translation:Vector=MakeVector(Y=MapRangeClamped(Value=@n3.Yaw, InRangeA=0, InRangeB=-30))) -> (Output_Get:Vector)
   n4 = Event OnCharacterCustomizationLoaded()
   n5 = Get Array Item with EFort Custom Part Type(Array:Array of Custom Character Part Object References=@n4.CustomParts) -> (Output:Custom Character Part Object Reference)
   > Set Is Female(Is Female:bool="Equal (Enum)"(
     A:EFortCustomGender Enum=Get GenderPermitted!(self:Custom Character Part Object Reference=@n5) -> (GenderPermitted:EFortCustomGender Enum=Invalid)
     B:EFortCustomGender Enum=Female
    )) -> (Output_Get:bool)
   n6 = Get BodyTypesPermitted!(self:Custom Character Part Object Reference=@n5) -> (BodyTypesPermitted:EFortCustomBodyType Enum=NONE)
   > Set Is Large(Is Large:bool="Equal (Enum)"(A:EFortCustomBodyType Enum=@n6, B:EFortCustomBodyType Enum=Large)) -> (Output_Get:bool)
   > Set Is Medium(Is Medium:bool="Equal (Enum)"(A=@n6, B=Medium)) -> (Output_Get:bool)
   > Set is F LRG(is F LRG:bool=AND Boolean(A=Get Is Large() -> (Is Large:bool), B=Get Is Female() -> (Is Female:bool))) -> (Output_Get:bool)
   > Set is F MED(is F MED:bool=AND Boolean(A=Get Is Medium() -> (Is Medium:bool), B=Get Is Female())) -> (Output_Get:bool)
   TryGetPawnOwner()
  ]}
]
variables [
 AmbientRot {type=Rotator, default=<rot 0 0 -50571.3672>}
 Backpack Translation {type=Vector}
 is F LRG {type=boolean}
 is F MED {type=boolean}
 Is Female {type=boolean}
 Is Large {type=boolean}
 Is Medium {type=boolean}
]
